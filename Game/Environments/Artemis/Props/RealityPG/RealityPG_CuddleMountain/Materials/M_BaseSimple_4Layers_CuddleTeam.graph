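# Nanite-enabled layered surface material for CuddleTeam mountain props, with three texture layers (third optional) despite the "4Layers" name.
# Layers sample by UV or pre-skinned-position biplanar projection, blended by a mask texture minus vertex paint or breakup, over a baked normal.
# Shared emissive, decal, gradient, dirt and up-vector functions follow, with cheaper Forward/Mobile branches, then snow progression and cinematic FX.
# A material proxy replace node supplies a simplified substitute.
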Material /Game/Environments/Artemis/Props/RealityPG/RealityPG_CuddleMountain/Materials/M_BaseSimple_4Layers_CuddleTeam
graph {
 properties {
  bAutomaticallySetUsageInEditor=false
  bUseMaterialAttributes=true
  bUsedWithNanite=true
  ditheredLODTransition=true
  pixelDepthOffsetMode=PDOM_Legacy
  refractionMethod=RM_IndexOfRefraction
  usedShadingModels=Default Lit
 }
 nodes [
  n1 = StaticBool() {value=true}
  n2 = Multiply(A=TextureCoordinate(), B=ScalarParameter() {defaultValue=1, group=Layer_01, parameterName=Layer1_UV_Tiling})
  n3 = TextureObjectParameter() {group=Layer_01, parameterName=Diffuse1, texture=/Game/Global/Textures/Default/Blanks/T_MidgreysRGB}
  n4 = TextureSample(UVs=@n2, Tex=@n3)
  n5 = TextureObjectParameter() {group=Layer_01, parameterName=Normals, samplerType=SAMPLERTYPE_Normal, texture=/Engine/EngineMaterials/FlatNormal}
  n6 = ScalarParameter() {defaultValue=2048, group=Layer_01, parameterName=Layer1_Tiling}
  n7 = MaterialFunctionCall!(
   ColorTex=@n3
   NormalTex=@n5
   Pos=NamedRerouteUsage()
   Nrm=NamedRerouteUsage()
   ColorUVScale=@n6
   ColorBlendFactor
   NormalStrength
   NormalUVScale=@n6
   NormalBlendFactor
   Local Space Inputs=NamedRerouteUsage()
  ) {materialFunction=/Game/Global/MaterialFunctions/MF_BiPlanar}
  n8 = Multiply(A=TextureCoordinate(), B=ScalarParameter() {defaultValue=1, group=Layer_02, parameterName=Layer2_UV_Tiling})
  n9 = TextureObjectParameter() {group=Layer_02, parameterName=Diffuse_02, texture=&T_MidgreysRGB}
  n10 = TextureObjectParameter() {group=Layer_02, parameterName=Normals_02, samplerType=SAMPLERTYPE_Normal, texture=&FlatNormal}
  n11 = ScalarParameter() {defaultValue=2048, group=Layer_02, parameterName=Layer2_Tiling}
  n12 = ScalarParameter() {defaultValue=12, group=Layer_02, parameterName=Layer_02_BlendFactor}
  n13 = MaterialFunctionCall(
   ColorTex=@n9
   NormalTex=@n10
   Pos=NamedRerouteUsage()
   Nrm=NamedRerouteUsage()
   ColorUVScale=@n11
   ColorBlendFactor=@n12
   NormalUVScale=@n11
   NormalBlendFactor=@n12
   Local Space Inputs=NamedRerouteUsage()
  ) {materialFunction=&MF_BiPlanar}
  n14 = LinearInterpolate(
   A=StaticSwitchParameter(True=@n4.RGBA, False=@n7.Color) {group=Layer_01, parameterName=UseUV}
   B=StaticSwitchParameter(True=TextureSample(UVs=@n8, Tex=@n9).RGBA, False=@n13.Color) {group=Layer_01, parameterName=UseUV}
   Alpha=NamedRerouteUsage()
  )
  n15 = Multiply(A=TextureCoordinate(), B=ScalarParameter() {defaultValue=1, group=Layer_03, parameterName=Layer3_UV_Tiling})
  n16 = TextureObjectParameter() {group=Layer_03, parameterName=Diffuse_03, texture=&T_MidgreysRGB}
  n17 = TextureObjectParameter() {group=Layer_03, parameterName=Normals_03, samplerType=SAMPLERTYPE_Normal, texture=&FlatNormal}
  n18 = ScalarParameter() {defaultValue=2048, group=Layer_03, parameterName=Layer3_Tiling}
  n19 = ScalarParameter() {defaultValue=12, group=Layer_03, parameterName=Layer_03_BlendFactor}
  n20 = MaterialFunctionCall(
   ColorTex=@n16
   NormalTex=@n17
   Pos=NamedRerouteUsage()
   Nrm=NamedRerouteUsage()
   ColorUVScale=@n18
   ColorBlendFactor=@n19
   NormalUVScale=@n18
   NormalBlendFactor=@n19
   Local Space Inputs=NamedRerouteUsage()
  ) {materialFunction=&MF_BiPlanar}
  n21 = StaticSwitchParameter(True=TextureSample(UVs=@n15, Tex=@n16).RGB, False=@n20.Color) {group=Layer_01, parameterName=UseUV}
  n22 = NamedRerouteUsage()
  n23 = TextureObjectParameter() {
   group=Layer_01
   parameterName=SpecularMasks
   samplerType=SAMPLERTYPE_Masks
   texture=&T_Fortnite_Default_S
  }
  n24 = TextureSample(UVs=@n2, Tex=@n23)
  n25 = NamedRerouteUsage()
  n26 = NamedRerouteUsage()
  n27 = ScalarParameter() {defaultValue=12, group=Layer_01, parameterName=Layer_01_BlendFactor}
  n28 = NamedRerouteUsage()
  n29 = TextureObjectParameter() {
   group=Layer_02
   parameterName=SpecularMasks_02
   samplerType=SAMPLERTYPE_Masks
   texture=&T_Fortnite_Default_S
  }
  n30 = NamedRerouteUsage()
  n31 = NamedRerouteUsage()
  n32 = NamedRerouteUsage()
  n33 = LinearInterpolate(
   A=StaticSwitchParameter(
    True=@n24.RGB
    False=MaterialFunctionCall(
     ColorTex=@n23
     Pos=@n25
     Nrm=@n26
     ColorUVScale=@n6
     ColorBlendFactor=@n27
     NormalUVScale=@n6
     NormalBlendFactor=@n27
     Local Space Inputs=@n28
    ) {materialFunction=&MF_BiPlanar}.Color
   ) {group=Layer_01, parameterName=UseUV}
   B=StaticSwitchParameter(
    True=TextureSample(UVs=@n8, Tex=@n29).RGB
    False=MaterialFunctionCall(ColorTex=@n29, Pos=@n30, Nrm=@n31, ColorUVScale=@n11, NormalUVScale=@n11, Local Space Inputs=@n32) {materialFunction=&MF_BiPlanar}.Color
   ) {group=Layer_01, parameterName=UseUV}
   Alpha=NamedRerouteUsage()
  )
  n34 = TextureObjectParameter() {
   group=Layer_03
   parameterName=SpecularMasks_03
   samplerType=SAMPLERTYPE_Masks
   texture=&T_Fortnite_Default_S
  }
  n35 = StaticSwitchParameter(
   True=TextureSample(UVs=@n15, Tex=@n34).RGB
   False=MaterialFunctionCall(
    ColorTex=@n34
    Pos=NamedRerouteUsage()
    Nrm=NamedRerouteUsage()
    ColorUVScale=@n18
    NormalUVScale=@n18
    Local Space Inputs=NamedRerouteUsage()
   ) {materialFunction=&MF_BiPlanar}.Color
  ) {group=Layer_01, parameterName=UseUV}
  n36 = NamedRerouteUsage()
  n37 = MaterialFunctionCall(Float3=StaticSwitchParameter(True=LinearInterpolate(A=@n33, B=@n35, Alpha=@n36), False=@n33) {defaultValue=true, group=Layer_03, parameterName=EnableLayer3}) {materialFunction=&BreakOutFloat3Components}
  n38 = TextureSample(UVs=@n2, Tex=@n5)
  n39 = LinearInterpolate(
   A=StaticSwitchParameter(True=@n38.RGB, False=@n7.Normals) {group=Layer_01, parameterName=UseUV}
   B=StaticSwitchParameter(True=TextureSample(UVs=@n8, Tex=@n10).RGB, False=@n13.Normals) {group=Layer_01, parameterName=UseUV}
   Alpha=NamedRerouteUsage()
  )
  n40 = LinearInterpolate(
   A=@n39
   B=StaticSwitchParameter(True=TextureSample(UVs=@n15, Tex=@n17).RGB, False=@n20.Normals) {group=Layer_01, parameterName=UseUV}
   Alpha=NamedRerouteUsage()
  )
  n41 = SetMaterialAttributes!(
   MaterialAttributes
   Base Color=StaticSwitchParameter(True=LinearInterpolate(A=@n14, B=@n21, Alpha=@n22), False=@n14) {defaultValue=true, group=Layer_03, parameterName=EnableLayer3}
   Specular=@n37.R
   Metallic=@n37.R
   Roughness=@n37.R
   Emissive Color
   Normal=MaterialFunctionCall(
    BaseNormal=TextureSampleParameter2D() {parameterName=BakedNoormal, samplerType=SAMPLERTYPE_Normal, texture=&FlatNormal}.RGB
    AdditionalNormal=StaticSwitchParameter(True=@n40, False=@n39) {defaultValue=true, group=Layer_03, parameterName=EnableLayer3}
   ) {materialFunction=&BlendAngleCorrectedNormals}.Result
   Custom Data 0
   Custom Data 1
   Opacity Mask
   Subsurface Color
   Opacity
   Refraction
  ) {attributeSetTypes [
    #1
    #2
    #3
    #4
    #5
    #6
    #7
    #8
    #9
    #10
    #11
    #12
   ]}
  n42 = MaterialFunctionCall!(In=@n41, Emissive Mask) {materialFunction=/Game/Athena/WinterArt/Materials/MF_EmissiveColor}
  n43 = MaterialFunctionCall!(In=@n42.Result) {materialFunction=/Game/Athena/WinterArt/Materials/MF_DecalPrint}
  n44 = MaterialFunctionCall!(transitionFactor, MaterialIn=MaterialFunctionCall!(In=@n43.Result) {materialFunction=/Game/Athena/WinterArt/Materials/MF_WS_Texture}.Result) {materialFunction=/Game/Athena/WinterArt/Materials/MF_ObjectSpaceGradient}
  n45 = MaterialFunctionCall!(
   In=MaterialFunctionCall!(
    UV=MaterialFunctionCall() {materialFunction=/Game/Athena/WinterArt/Materials/MF_UVsSet}.UVs 4 Layers
    In=MaterialFunctionCall!(transitionFactor, MaterialIn=@n44.Color*Gradient) {materialFunction=/Game/Athena/WinterArt/Materials/MF_WSGradient}.Color*Gradient
   ) {materialFunction=/Game/Athena/WinterArt/Materials/MF_WSDirt}.Result
   UseOneSampler
  ) {materialFunction=/Game/Athena/WinterArt/Materials/MF_VPTexture}
  n46 = MaterialFunctionCall!(In=@n45.Result, UpVectorTexture=@n45.Result) {materialFunction=/Game/Athena/WinterArt/Materials/MF_UpVectorTexture}
  n47 = MaterialFunctionCall(In=MaterialFunctionCall!(In=@n43.Result, Emissive Mask=@n42.EmissiveMask) {materialFunction=/Game/Athena/WinterArt/Materials/MF_Screens}.Result) {materialFunction=&MF_VPTexture}
  n48 = MaterialFunctionCall(In=@n47.Result, UpVectorTexture=@n47.Result) {materialFunction=&MF_UpVectorTexture}
  n49 = GetMaterialAttributes(None=@n48.Result) {attributeGetTypes [
    #1
    #5
    #9
    #2
    #4
    #3
   ]}
  n50 = SetMaterialAttributes!(
   MaterialAttributes
   Base Color=@n49.Base Color
   Emissive Color=@n49.Base Color
   Opacity Mask=@n49.Base Color
   Specular=Multiply(A=@n49.Specular, B=ScalarParameter() {defaultValue=1, group=Mobile, parameterName=Mobile_SpecularMult})
   Roughness=Multiply(A=@n49.Roughness, B=ScalarParameter() {defaultValue=1, group=Mobile, parameterName=Mobile_RoughnessMult})
   Metallic=Multiply(A=@n49.Metallic, B=ScalarParameter() {defaultValue=1, group=Mobile, parameterName=Mobile_MetalMult})
   Subsurface Color=@n49.Base Color
  ) {attributeSetTypes [
    #1
    #5
    #9
    #2
    #4
    #3
    #10
   ]}
  n51 = GetMaterialAttributes(None=@n48.Result) {attributeGetTypes [
    #1
    #5
    #9
    #2
    #4
    #3
    #6
   ]}
  n52 = SetMaterialAttributes!(
   MaterialAttributes
   Base Color=@n51.Base Color
   Emissive Color=@n51.Base Color
   Opacity Mask=@n51.Base Color
   Specular=Multiply(A=@n51.Specular, B=ScalarParameter() {defaultValue=1, group=Mobile, parameterName=Mobile_SpecularMult})
   Roughness=Multiply(A=@n51.Roughness, B=ScalarParameter() {defaultValue=1, group=Mobile, parameterName=Mobile_RoughnessMult})
   Metallic=Multiply(A=@n51.Metallic, B=ScalarParameter() {defaultValue=1, group=Mobile, parameterName=Mobile_MetalMult})
   Normal=@n51.Base Color
   Subsurface Color=@n51.Base Color
  ) {attributeSetTypes [
    #1
    #5
    #9
    #2
    #4
    #3
    #6
    #10
   ]}
  n53 = MaterialFunctionCall!(
   IsStaticSnow
   UseAlternateStaticSnowMask
   AlternateStaticSnowMask
   "UseRoad&&OcclusionMask"
   IsForGrassMeshToPushDown
   AdditionalMaskToRemoveSnow
   IsFallingSnowFX
   UseMinimapCoords
   IncreaseSnowUsingSlope
   UseSmoothBorder
   LerpHighM
   TemporalHeightLerpContrast
   IsLobbyIsland
   UseBR-IslandMask=StaticBool()
   IsTree
   IsForRenderTarget
   UseSparkle
   UseSnowNormals=@n1
   UseAdditionalMaskToRemoveSnow
   IsTangentSpaceInput=@n1
   UseVertexNormalsForBasicSnow
   UseTangentSnowNormals=@n1
   Material=ShadingPathSwitch(
    Default=MaterialFunctionCall!(In=@n46.Result) {materialFunction=/Game/Athena/WinterArt/Materials/MF_RandomDecalSplatter}.Result
    Forward=@n48.Result
    Mobile=DistanceFieldsRenderingSwitch!(No=@n50, Yes=@n52)
   )
  ) {
   desc=Realtime
   materialFunction=/Game/Athena/Apollo/Environments/Landscape/Materials/Biome/UniqueBiomeMaterialFunctions/MF_Apollo_Snow_Progression
  }
  n54 = GetMaterialAttributes(None=MaterialFunctionCall!(In=@n53."DefaultSwitch=FalseMaterial") {materialFunction=/Game/Athena/WinterArt/Materials/MF_GameWideShared}.Result) {attributeGetTypes [#6]}
  MaterialFunctionCall!(In=SetMaterialAttributes!(MaterialAttributes=@n54.MaterialAttributes, Normal=@n54.MaterialAttributes) {attributeSetTypes [#6]}) {materialFunction=/Game/Packages/Fortress_SharedMaterials/Cinematics/CINE_MF_EnvPostFX}
  MaterialProxyReplace(Realtime=@n41, MaterialProxy=NamedRerouteUsage())
  n55 = TextureObjectParameter() {
   group=LayerBlending
   parameterName=LayerBlendTexture
   samplerType=SAMPLERTYPE_Masks
   texture=&T_Fortnite_Default_S
  }
  n56 = NamedRerouteUsage()
  n57 = NamedRerouteUsage()
  n58 = ScalarParameter() {defaultValue=2048, group=LayerBlending, parameterName=Layer_01BlendTextureTiling}
  n59 = ScalarParameter() {defaultValue=12, group=LayerBlending, parameterName=Layer_01BlendTextureBlending}
  n60 = NamedRerouteUsage()
  n61 = VertexColor()
  n62 = ChannelMaskParameter(None=MaterialFunctionCall(
    ColorTex=@n55
    Pos=NamedRerouteUsage()
    Nrm=NamedRerouteUsage()
    ColorUVScale=ScalarParameter() {defaultValue=2048, group=LayerBlending, parameterName=Layer_01BlendBreakupTextureTiling}
    ColorBlendFactor=ScalarParameter() {defaultValue=12, group=LayerBlending, parameterName=Layer_01BreakupTextureBlending}
    Local Space Inputs=NamedRerouteUsage()
   ) {materialFunction=&MF_BiPlanar}.Color) {group=LayerBlending, maskChannel=Blue, parameterName=LayerBlendBreakupChannel, defaultValue=<rgba 0 0 1 0>}
  n63 = Clamp(None=Subtract(
    A=MaterialFunctionCall(ColorTex=@n55, Pos=@n56, Nrm=@n57, ColorUVScale=@n58, ColorBlendFactor=@n59, Local Space Inputs=@n60) {materialFunction=&MF_BiPlanar}.Color
    B=StaticSwitchParameter(True=@n61, False=@n62.R) {group=LayerBlending, parameterName=UseVertexPaintForBreakup}
   ))
  NamedRerouteDeclaration(None=ChannelMaskParameter(None=@n63) {group=LayerBlending, parameterName=Layer2BlendChannel}.R) {name=Layer2BlendMask, nodeColor=<rgba 1 0 0.1406 1>}
  NamedRerouteDeclaration(None=ChannelMaskParameter(None=@n63) {group=LayerBlending, parameterName=Layer3BlendChannel}.R) {name=Layer3BlendMask, nodeColor=<rgba 0 1 0.6953 1>}
  n64 = MaterialFunctionCall(Float3=@n24.RGB) {materialFunction=&BreakOutFloat3Components}
  n65 = SetMaterialAttributes!(
   MaterialAttributes
   Base Color=Multiply(A=@n4.RGB, B=ScalarParameter() {defaultValue=1, parameterName=HLOD_DarkenDiffuse})
   Specular=@n64.R
   Metallic=@n64.R
   Roughness=@n64.R
   Normal=@n38.RGB
  ) {attributeSetTypes [
    #1
    #2
    #3
    #4
    #6
   ]}
  NamedRerouteDeclaration(None=@n65) {name=HLOD_Attrs, nodeColor=<rgba 0 0.25 1 1>}
  n66 = PreSkinnedPosition()
  NamedRerouteDeclaration(None=ShaderStageSwitch(PixelShader=VertexInterpolator(VS=@n66).PS, VertexShader=@n66)) {name=PreskinnedPos, nodeColor=<rgba 1 0 0.4219 1>}
  n67 = PreSkinnedNormal()
  NamedRerouteDeclaration(None=ShaderStageSwitch(PixelShader=VertexInterpolator(VS=@n67).PS, VertexShader=@n67)) {name=PreskinnedNormal, nodeColor=<rgba 0.2422 1 0 1>}
  NamedRerouteDeclaration(None=StaticBoolParameter() {defaultValue=true, parameterName=UseLocalPos}) {name=UseLocalPos, nodeColor=<rgba 1 0.4453 0 1>}
  Transform(None=@n54.Normal)
  Transform(None=@n7.Normals)
 ]
}
